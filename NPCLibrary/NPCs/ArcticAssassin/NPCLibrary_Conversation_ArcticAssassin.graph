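# Conversation database for the Arctic Assassin NPC, a shy, friend-seeking character that players can talk to.
# A single entry point leads to a speech node whose default line can be replaced by contextual greetings for a first-ever meeting or specific player outfits, ranked by curve-driven priority.
# The speech node then branches through link nodes to shared NPC quests, basic services and a Season 26 Snapshot quest conversation.
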ConversationDatabase /NPCLibrary/NPCs/ArcticAssassin/NPCLibrary_Conversation_ArcticAssassin
entry_points [{tag=Conversation.Entry.ArcticAssassin, destinations [#1]}]
nodes [
 ConversationEntryPointNode_7 {
  id=#1
  class=ConversationEntryPointNode
  outputs [#2]
  properties {entryTag {tagName=Conversation.Entry.ArcticAssassin}}
 }
 Speech_C_4 {
  id=#2
  class=Speech_C
  outputs [#3, #4, #5]
  properties {
   nodeName=""
   generalConfig {
    defaultMessage="I know I'm an assassin, but I'm also a really nice person looking for a friend."
    contextualMessages [
     {
      message="Hi! It's so lovely to meet you! Please don't run away."
      contextRequirements [{requirement {class=FortControllerRequirement_IsFirstEverConversationWithNPC}}]
      priority {curve {rowName=Tandem.Speech.Priority.FirstEverInteraction}}
     }
     {
      message=Do you want to hit yoga together? I'm trying to make some friends...
      contextRequirements [{requirement {
         class=FortControllerRequirement_HasCID
         properties {allowedCIds [/BRCosmetics/Athena/Items/Cosmetics/Characters/CID_189_Athena_Commando_F_Exercise]}
        }}]
      priority {curve {rowName=Tandem.Speech.Priority.OutfitReaction}}
     }
     {
      message="I've never been to space! Is it, um... is it cool? We could go sometime..."
      contextRequirements [{requirement {
         class=FortControllerRequirement_HasCID
         properties {allowedCIds [/BRCosmetics/Athena/Items/Cosmetics/Characters/CID_477_Athena_Commando_F_SpaceSuit]}
        }}]
      priority {curve {rowName=Tandem.Speech.Priority.OutfitReaction}}
     }
     {
      message="Those clothes look really comfy! Where did you, um... Sorry, I'm bad at, uh... talking to people."
      contextRequirements [{requirement {
         class=FortControllerRequirement_HasCID
         properties {allowedCIds [/BRCosmetics/Athena/Items/Cosmetics/Characters/CID_824_Athena_Commando_F_RaiderPink]}
        }}]
      priority {curve {rowName=Tandem.Speech.Priority.OutfitReaction}}
     }
     {
      message="That suit is SO NICE! Umm... Do you want to be friends?"
      contextRequirements [{requirement {
         class=FortControllerRequirement_HasCID
         properties {allowedCIds [/BRCosmetics/Athena/Items/Cosmetics/Characters/CID_722_Athena_Commando_M_TeriyakiFishAssassin]}
        }}]
      priority {curve {rowName=Tandem.Speech.Priority.OutfitReaction}}
     }
    ]
   }
  }
 }
 ConversationLinkNode_2 {
  id=#5
  class=ConversationLinkNode
  properties {remoteEntryTag {tagName=Conversation.Entry.Season26.Snapshot.P01.Q012.c}}
 }
 ConversationLinkNode_0 {id=#4, class=ConversationLinkNode, properties {remoteEntryTag {tagName=Conversation.Entry.BasicServices}}}
 ConversationLinkNode_1 {id=#3, class=ConversationLinkNode, properties {remoteEntryTag {tagName=Conversation.Entry.NPCLibrary.Quests}}}
]
